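# Character material for the GlacialTrooper medium male player body, adding a volumetric ice effect to the shared 2023 character shading.
# It chains the character parent PreFX function through MF_VolumetricIce into PostFX, using material attributes.
# The PreFX tech art mask reaches the ice function through a named reroute.
# Shading model comes from the graph, covering lit, subsurface, skin profile, clear coat and cloth, with skeletal mesh, morph target and deformer support.
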Material /BRCosmetics/Characters/Player/Male/Medium/Bodies/M_MED_GlacialTrooper/Materials/M_GlacialTrooper
graph {
 properties {
  bAutomaticallySetUsageInEditor=false
  bUseMaterialAttributes=true
  bUsedWithMeshDeformer=true
  bUsedWithMorphTargets=true
  bUsedWithSkeletalMesh=true
  shadingModel=MSM_FromMaterialExpression
  subsurfaceProfile=&SS_HeroSkin_02
  usedShadingModels=Default Lit | Subsurface | Clear Coat | Subsurface Profile | Cloth
 }
 nodes [
  n1 = MaterialFunctionCall() {materialFunction=&MF_CharacterParent_2023_PreFX}
  MaterialFunctionCall(
   Material=MaterialFunctionCall!(Material=@n1.Result, Tech Art Mask=NamedRerouteUsage()) {materialFunction=&/MF_VolumetricIce}."Result (MA)"
   M=@n1.M
   Pre Skinned Local Position=@n1.M
   Distance Blend=@n1.M
   Fresnel=@n1.M
  ) {materialFunction=&MF_CharacterParent_2023_PostFX}
  NamedRerouteDeclaration(None=@n1.Tech Art Mask) {name="Tech Art Mask (V3)", nodeColor=<rgba 0.6406 0 1 1>}
 ]
}
